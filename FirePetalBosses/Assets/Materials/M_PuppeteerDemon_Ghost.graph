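# Translucent, unlit skeletal-mesh material that gives the Puppeteer Demon boss a smoky ghost look with a brighter core.
# Scrolling smoke and distortion textures, Fresnel, core sphere-mask and atmospheric-light gradients provide the masks, with pre-skinned bounds-normalized coordinates available.
# Curve-atlas colors, sky and inner light tints blend into emissive, while opacity passes through a translucency-sorting function.
# Time-driven sine, cosine and noise terms build waving world position offsets.
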Material /FirePetalBosses/Assets/Materials/M_PuppeteerDemon_Ghost
graph {
 properties {
  allowTranslucentCustomDepthWrites=true
  bAutomaticallySetUsageInEditor=false
  bHasPixelAnimation=true
  bUsedWithMeshDeformer=true
  bUsedWithSkeletalMesh=true
  blendMode=BLEND_Translucent
  shadingModel=MSM_Unlit
  usedShadingModels=Unlit
 }
 nodes [
  n1 = VectorParameter() {parameterName=SkyLightColor, defaultValue=<rgba 0.0497 0.0052 0.0331 0.502>}
  n2 = NamedRerouteUsage()
  n3 = Add(A=@n2) {constB=-0.5}
  n4 = Multiply(A=@n1.RGB, B=Saturate(None=Multiply(A=@n3) {constB=1.33}))
  n5 = CurveAtlasRowParameter(CurveTime=ComponentMask(None=@n2) {r=true}) {
   atlas=/FirePetalBosses/Assets/Textures/T_DemonGhost_Atlas
   curve=/FirePetalBosses/Assets/Textures/CC_DemonGhost_Base
   parameterName=PrimaryColorCurve
  }
  n6 = Divide(A=Add(A=Multiply(A=@n4, B=@n1.A), B=@n5)) {constB=1.5}
  Add(
   A=Multiply(A=VectorParameter() {parameterName=InnerLightColor, defaultValue=<rgba 0.075 0.0078 0.0499 0.502>}.RGB, B=NamedRerouteUsage())
   B=@n6
  )
  n7 = Power(Base=Multiply(A=NamedRerouteUsage()) {constB=2.5}, Exp=ScalarParameter() {defaultValue=1, parameterName=MaskPower})
  MaterialFunctionCall!(Opacity=Multiply(A=Saturate(None=@n7), B=ScalarParameter() {defaultValue=1, parameterName=OverallOpacity})) {materialFunction=/Game/Global/Materials/MaterialGym/Materials/SortedOutline/MF_SortTranslucency}
  n8 = Multiply(A=Multiply(A=VectorParameter() {parameterName=InnerLightColor, defaultValue=<rgba 0.075 0.0078 0.0499 0.502>}.RGB, B=NamedRerouteUsage())) {constB=0.333}
  n9 = Subtract(A=Multiply(A=NamedRerouteUsage(), B=NamedRerouteUsage()), B=ComponentMask(None=NamedRerouteUsage()) {r=true})
  n10 = CurveAtlasRowParameter(CurveTime=Saturate(None=Power(Base=@n9) {constExponent=3})) {
   atlas=&T_DemonGhost_Atlas
   curve=/FirePetalBosses/Assets/Textures/CC_DemonGhost_Core
   defaultValue=1
   parameterName=SecondaryColorCurve
  }
  n11 = Power(Base=Multiply(A=@n10, B=ScalarParameter() {defaultValue=3, parameterName=CoreEmissiveBrightness}))
  n12 = OneMinus(None=NamedRerouteUsage())
  MaterialFunctionCall(Emissive=Add(A=Saturate(None=@n8), B=LinearInterpolate(A=@n6, B=@n11, Alpha=@n12))) {materialFunction=&MF_FX_EmissiveAdjustments}
  n13 = NamedRerouteUsage()
  n14 = Multiply(A=MaterialFunctionCall(Float3=OneMinus(None=@n13)) {materialFunction=&BreakOutFloat3Components}.B) {constB=3}
  n15 = Time()
  n16 = Add(A=Multiply(A=Add(A=@n14, B=@n15), B=ScalarParameter() {defaultValue=0.2, parameterName=WPO_Speed}), B=NamedRerouteUsage())
  n17 = ComponentMask(None=@n16) {r=true}
  n18 = ComponentMask(None=@n16) {g=true}
  n19 = Multiply(
   A=MaterialFunctionCall(X=Cosine(None=@n17), Y=Sine(None=@n18)) {materialFunction=&MakeFloat3}.Result
   B=Saturate(None=NamedRerouteUsage())
  )
  n20 = MaterialFunctionCall(Float3=NamedRerouteUsage()) {materialFunction=&BreakOutFloat3Components}
  Multiply(
   A=Multiply(A=Multiply(A=@n19, B=@n20.B), B=ScalarParameter() {defaultValue=2, parameterName=WPO_Strength})
   B=Constant3Vector() {constant=<rgba 1 1 0 0>}
  )
  n21 = LocalPosition()
  n22 = VertexInterpolator(VS=@n21.XYZ)
  n23 = ObjectLocalBounds()
  n24 = Subtract(
   A=NamedRerouteDeclaration(None=ShaderStageSwitch(PixelShader=@n22.PS, VertexShader=@n21.XYZ)) {name=PreSkinnedPosition, nodeColor=<rgba 0 1 0.6484 1>}
   B=@n23.Min
  )
  n25 = MaterialFunctionCall(Color=Subtract(A=@n23.Max, B=@n23.Max)) {materialFunction=&SplitComponents}
  n26 = Max(A=Max(A=@n25.R, B=@n25.R), B=@n25.B)
  NamedRerouteDeclaration(None=Multiply(A=Divide(A=@n24, B=@n26), B=VectorParameter() {parameterName=BoundsScale, defaultValue=<rgba 1 1 1 1>}.RGB)) {name=UVBounds, nodeColor=<rgba 0.5469 1 0 1>}
  n27 = Time()
  n28 = VectorParameter() {parameterName=DistortionSpeed, defaultValue=<rgba 0.005 0.005 0.03 0>}
  n29 = MaterialFunctionCall(
   TextureObject=TextureObject() {samplerType=SAMPLERTYPE_Normal, texture=/Game/Global/Textures/Default/Tileable_Patterns/2ndInfectedMapNormal}
   TextureSize=Constant() {r=1}
   WorldPosition=Add(A=NamedRerouteUsage(), B=Multiply(A=@n27, B=@n28.RGB))
  ) {materialFunction=&WorldAlignedTexture}
  NamedRerouteDeclaration(None=Multiply(A=ScalarParameter() {defaultValue=0.025, parameterName=DistortAmount}, B=@n29.XYZ Texture)) {name=DistortionNormal, nodeColor=<rgba 0.3594 1 0 1>}
  n30 = DotProduct(A=CameraVectorWS(), B=VertexNormalWS())
  n31 = Power(Base=@n30, Exp=ScalarParameter() {defaultValue=4, parameterName=CoreFresnelPower})
  n32 = Multiply(A=OneMinus(None=NamedRerouteUsage())) {constB=0.7}
  NamedRerouteDeclaration(None=Saturate(None=Multiply(A=@n31, B=@n32))) {name=CoreFresnel, nodeColor=<rgba 0 1 0.625 1>}
  n33 = NamedRerouteUsage()
  n34 = Constant3Vector() {constant=<rgba 0.4 0.3 0.7 0>}
  NamedRerouteDeclaration(None=MaterialFunctionCall!(In=SphereMask(A=@n33, B=@n34) {attenuationRadius=0.5, hardnessPercent=0}) {materialFunction=/Game/Effects/Fort_Effects/Materials/_Functions/MF_Smoothstep_01}.Result) {name=CoreGradient, nodeColor=<rgba 0.2891 0 1 1>}
  n35 = VertexNormalWS()
  n36 = NamedRerouteUsage()
  n37 = NamedRerouteDeclaration(None=DotProduct(A=AtmosphericLightVector(), B=Add(A=@n35, B=@n36))) {name=NDotL, nodeColor=<rgba 0 1 0.8125 1>}
  n38 = ComponentMask(None=NamedRerouteUsage()) {r=true}
  n39 = Constant() {r=0.6}
  n40 = ConstantBiasScale(None=Subtract(A=ConstantBiasScale(None=@n37), B=Multiply(A=@n38, B=@n39))) {bias=0.666, scale=0.75}
  n41 = Multiply(A=NamedRerouteUsage()) {constB=0.25}
  NamedRerouteDeclaration(None=Saturate(None=Subtract(A=@n40, B=@n41))) {name=LightGradient, nodeColor=<rgba 0 1 0.1797 1>}
  n42 = MaterialFunctionCall!(Thickness, NormalWS, PowerExponent=ScalarParameter() {defaultValue=2, parameterName=FresnelPower}) {materialFunction=/Game/Weapons/Prototype/ObjectMover/Materials/MF_NormalizedFresnel}
  n43 = NamedRerouteUsage()
  n44 = Subtract(
   A=NamedRerouteDeclaration(None=@n42.Power) {name=Fresnel, nodeColor=<rgba 1 0.4688 0 1>}
   B=Multiply(A=ComponentMask(None=@n43) {r=true}, B=ScalarParameter() {defaultValue=0.5, parameterName=FresnelNoiseStrength})
  )
  n45 = Saturate(None=Subtract(A=@n42.Power, B=ScalarParameter() {defaultValue=1, parameterName=FresnelBase}) {constB=0.8})
  NamedRerouteDeclaration(None=Saturate(None=Add(A=@n44, B=@n45))) {name=SmokeFresnel, nodeColor=<rgba 0.9688 1 0 1>}
  n46 = TextureObjectParameter() {parameterName=SmokeTexture, samplerType=SAMPLERTYPE_LinearGrayscale, texture=/FirePetalBosses/Assets/Textures/T_DemonSmokeNoise}
  n47 = Add(
   A=NamedRerouteUsage()
   B=Multiply(A=Time(), B=VectorParameter() {parameterName=PrimarySmokeSpeed, defaultValue=<rgba -0.025 0.025 -0.025 0>}.RGB)
  )
  n48 = NamedRerouteUsage()
  NamedRerouteDeclaration(None=MaterialFunctionCall(TextureObject=@n46, TextureSize=ScalarParameter() {defaultValue=0.5, parameterName=SmokeTextureSize}, WorldPosition=Add(A=@n47, B=@n48)) {materialFunction=&WorldAlignedTexture}.XYZ Texture) {name=PrimarySmoke, nodeColor=<rgba 0 0.9766 1 1>}
  n49 = DotProduct(A=VertexNormalWS(), B=Constant3Vector() {constant=<rgba 0 0 1 0>})
  n50 = NamedRerouteUsage()
  NamedRerouteDeclaration(None=Saturate(None=Multiply(A=@n49, B=@n50))) {name=UpDirection, nodeColor=<rgba 1 0 0.4219 1>}
  n51 = MaterialFunctionCall(Float3=@n29.XYZ Texture) {materialFunction=&BreakOutFloat3Components}
  NamedRerouteDeclaration(None=Max(A=Max(A=@n51.R, B=@n51.R), B=@n51.B)) {name=DistortGradient, nodeColor=<rgba 0 0.7187 1 1>}
  n52 = NamedRerouteUsage()
  n53 = Panner(
   Coordinate=Multiply(A=MaterialFunctionCall(Float3=@n52) {materialFunction=&BreakOutFloat3Components}.B) {constB=1.5}
   Time=Time()
  ) {speedX=0.05, speedY=-0.1}
  NamedRerouteDeclaration(None=Multiply(A=TextureSample(UVs=@n53) {texture=/Game/Global/Textures/Default/Noise/3xPerlin_M}.RGB) {constB=0.5}) {name=WPO_Distort, nodeColor=<rgba 1 0 0.1641 1>}
  n54 = VertexNormalWS()
  n55 = WorldPosition() {worldPositionShaderOffset=WPT_ExcludeAllShaderOffsets}
  n56 = ScalarParameter() {defaultValue=0.01, parameterName=WorldWaveWPOFrequency}
  n57 = Subtract(A=Frac(None=Multiply(A=@n55.XYZ, B=@n56) {constB=0.0025})) {constB=0.5}
  n58 = MaterialFunctionCall(Float3=Abs(None=Multiply(A=@n57) {constB=2})) {materialFunction=&BreakOutFloat3Components}
  n59 = Subtract(A=Add(A=Add(A=@n58.R, B=@n58.R), B=@n58.B)) {constB=2}
  n60 = Frac(None=Multiply(A=Time()) {constB=0.3})
  n61 = Add(A=Sine(None=Subtract(A=@n59, B=@n60)))
  n62 = MaterialFunctionCall(Float3=Multiply(A=NamedRerouteUsage(), B=ScalarParameter() {defaultValue=0.075, parameterName=LocalWaveWPOFrequency}) {constB=0.02}) {materialFunction=&BreakOutFloat3Components}
  n63 = Add(A=Add(A=@n62.R, B=@n62.R), B=@n62.B)
  n64 = Multiply(A=Time()) {constB=0.75}
  n65 = Sine(None=Subtract(A=Subtract(A=@n63) {constB=2}, B=Frac(None=@n64)))
  n66 = Add(
   A=Multiply(A=Multiply(A=@n61) {constB=0.5}, B=ScalarParameter() {defaultValue=2, parameterName=WPOWorldScale})
   B=Multiply(A=Add(A=@n65)) {constB=0.5}
  ) {constB=0}
  NamedRerouteDeclaration(None=Multiply(A=Multiply(A=@n54, B=@n66), B=ScalarParameter() {defaultValue=0.25, parameterName=WPO_FinalScale})) {name=WPO_Waving, nodeColor=<rgba 1 0.3281 0 1>}
  NamedRerouteUsage()
  NamedRerouteUsage()
  NamedRerouteUsage()
  Power(
   Base=Multiply(A=@n30, B=ScalarParameter() {defaultValue=1, parameterName=FresnelBias})
   Exp=ScalarParameter() {defaultValue=2, parameterName=FresnelPower}
  )
  PreSkinnedLocalBounds()
  VertexNormalWS()
 ]
}
